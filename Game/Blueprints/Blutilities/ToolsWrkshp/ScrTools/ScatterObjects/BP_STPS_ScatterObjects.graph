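# Editor utility Blueprint that holds the settings for an object scatter tool.
# Variables define the objects to scatter, shape, amount, size, rotation and scale ranges, vertical offset, overlap, actor type, label prefix and shape preview color.
# The Clear Current ISM Actor function only broadcasts the EventClearActorClicked dispatcher, leaving the clearing to whatever binds it.
# The event graph is empty, so no scattering logic lives here.
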EditorUtilityBlueprint /Game/Blueprints/Blutilities/ToolsWrkshp/ScrTools/ScatterObjects/BP_STPS_ScatterObjects
graphs [
 Clear Current ISM Actor {nodes [
   Clear Current ISM Actor() {is=title K2Node_FunctionEntry}
   > Call Event Clear Actor Clicked!(execute:Exec, self:BP STPS Scatter Objects Object Reference) -> (then:Exec) {is=title K2Node_CallDelegate}
  ]}
 EventClearActorClicked {nodes [EventClearActorClicked() {is=title K2Node_FunctionEntry}]}
 EventGraph {links [], nodes []}
]
variables [
 Actor Type {type=E_ActorType_ScatterObjects, default=Default}
 Allow Overlap {type=boolean, default=true}
 Base Shape {type=E_BaseShapes_ScatterObjects, default=Noisy Circle}
 Base Shape Opacity {type=number, default=0.4}
 EventClearActorClicked {default=()}
 Label Prefix {type=string, default=Scatter Tool}
 Objects To Scatter {type="array<S_SelectionChoices_ScatterObjects>", default []}
 Rotation Range {type=Vector2D, default=<-180 180>}
 Scale Range {type=Vector2D, default=<1 1>}
 Scatter Amount {type=integer, default=5}
 Scatter Shape Color {type=LinearColor, default=<rgba 1 1 1 0.05>}
 Scatter Shape Size {type=Vector2D, default=<2048 2048>}
 Vertical Offset {type=number}
]
