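# Editor-only scriptable tool that scatters static meshes as ISM or HISM instances on a tagged level actor, with undo transactions.
# Setup adds property sets, click, hover and hotkey behaviors, plus header and dropper-overlay widgets.
# Shape settings rebuild a transient Geometry Script preview: rectangle, ellipse or Perlin-noised disc.
# Helpers reuse or create per-mesh components via SubobjectDataSubsystem, check material instancing support and trace clicks with landscape-aware filtering.
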EditorUtilityBlueprint /Game/Blueprints/Blutilities/ToolsWrkshp/ScrTools/ScatterObjects/BP_ST_ScatterObjects
graphs [
 Check Material Validity {nodes [
   n1 = Check Material Validity!() -> (then, Assets To Scatter:Array of S Selection Choices Scatter Objects Structures) {is=title K2Node_FunctionEntry}
   > n2 = For Each Loop(Array:Array of S Selection Choices Scatter Objects Structures=@n1) -> (Array Element:S Selection Choices Scatter Objects Structure)
   Return Node!(
    execute=@n2.Completed
    Assets Instanceable:Array of S Selection Choices Scatter Objects Structures=Get Assets With Instanceable Materials() -> (Assets With Instanceable Materials:Array of S Selection Choices Scatter Objects Structures)
    Assets Non Instanceable:Array of S Selection Choices Scatter Objects Structures=Get Assets With Non Instanceable Materials() -> (Assets With Non Instanceable Materials:Array of S Selection Choices Scatter Objects Structures)
   )
   > n3 = Has Instanceable Material!(
    execute
    self
    Static Mesh:Static Mesh Object Reference=Break S Selection Choices Scatter Objects!(S_SelectionChoices_ScatterObjects:"S Selection Choices Scatter Objects Structure (by ref)"=@n2) -> (Mesh_15_#1:Static Mesh Object Reference, Weight_5_#2:int) {is=title K2Node_BreakStruct}
   ) -> (then, Material Instanceable:bool)
   > n4 = Branch(Condition=@n3)
   > Add Unique(
    TargetArray:Array of S Selection Choices Scatter Objects Structures=Get Assets With Instanceable Materials()
    NewItem:"S Selection Choices Scatter Objects Structure (by ref)"
   )
   Add Unique(execute=@n4.else, TargetArray=Get Assets With Non Instanceable Materials(), NewItem=@n2)
  ]}
 Create Scattered HISM {nodes [
   n1 = Create Scattered HISM!() -> (then, Scattered Assets:Array of S Inst Data Scatter Objects Structures) {is=title K2Node_FunctionEntry}
   > BeginTransaction(Context=Get Undo Clear Context() -> (Undo Clear Context:String), Description="LOCTEXT(\"Create Actor\")")
   > TransactObject(Object=Get ISM Actor() -> (ISM Actor:Actor Object Reference))
   > n2 = For Each Loop(Array:Array of S Inst Data Scatter Objects Structures=@n1) -> (Array Element:S Inst Data Scatter Objects Structure)
   EndTransaction(execute=@n2.Completed)
   n3 = Break S Inst Data Scatter Objects!(S_InstData_ScatterObjects:"S Inst Data Scatter Objects Structure (by ref)"=@n2) -> (MeshToScatter_8_#1:Static Mesh Object Reference, InstTransform_14_#3:Transform) {is=title K2Node_BreakStruct}
   n4 = Find Or Add HISM Component!(execute=@n2.LoopBody, self, SM In Instance:Static Mesh Object Reference=@n3) -> (then, HISM Component:Hierarchical Instanced Static Mesh Component Object Reference)
   > TransactObject(Object=@n4)
   > AddInstance!(
    execute
    self:Instanced Static Mesh Component Object Reference=@n4
    InstanceTransform:"Transform (by ref)"=@n3.InstTransform_14_#3
    bWorldSpace:bool=true
   ) -> (then, ReturnValue:int=0)
  ]}
 Create Scattered ISM {nodes [
   n1 = Create Scattered ISM!() -> (then, Scattered Assets:Array of S Inst Data Scatter Objects Structures) {is=title K2Node_FunctionEntry}
   > BeginTransaction(Context=Get Undo Clear Context(), Description="LOCTEXT(\"Create Actor\")")
   > TransactObject(Object=Get ISM Actor())
   > n2 = For Each Loop(Array=@n1)
   EndTransaction(execute=@n2.Completed)
   n3 = Break S Inst Data Scatter Objects(S_InstData_ScatterObjects=@n2)
   n4 = Find Or Add ISM Component!(execute=@n2.LoopBody, self, SM In Instance:Static Mesh Object Reference=@n3) -> (then, ISM Component:Instanced Static Mesh Component Object Reference)
   > TransactObject(Object=@n4)
   > AddInstance(self=@n4, InstanceTransform=@n3.InstTransform_14_#3, bWorldSpace=true) -> (ReturnValue=0)
   "To String (Text)"(InText=Get ToolName() -> (ToolName:Text))
  ]}
 EventGraph {nodes [
   Event OnScriptSetup()
   > n1 = Sequence!(execute) -> (then_0, then_1, then_2:Exec, then_3:Exec, then_4:Exec, then_5:Exec, then_6:Exec, then_7:Exec, then_8:Exec, then_9:Exec)
   > K2Node_Composite_0 = Add Property Sets() {is=title K2Node_Composite}
   K2Node_Composite_2 = Watch Properties(execute=@n1.then_2) {is=title K2Node_Composite}
   K2Node_Composite_4 = Bind To Events(execute=@n1.then_3) {is=title K2Node_Composite}
   K2Node_Composite_1 = Add Behaviours(execute=@n1.then_4) {is=title K2Node_Composite}
   K2Node_Composite_5 = Setup Const Variables(execute=@n1.then_5) {is=title K2Node_Composite}
   K2Node_Composite_6 = Add UserWidgets(execute=@n1.then_6) {is=title K2Node_Composite}
   K2Node_Composite_7 = Find Selected ISM Actor(execute=@n1.then_7) {is=title K2Node_Composite}
   K2Node_Composite_8 = Init Tool State(execute=@n1.then_8) {is=title K2Node_Composite}
   Event OnScriptShutdown()
   > SavePropertySetSettings(
    PropertySet=Get Settings() -> (Settings:BP STPS Scatter Objects Object Reference)
    SaveKey=Get Prop Set Save Key() -> (Prop Set Save Key:String)
   )
   > n2 = Sequence!(execute) -> (then_0, then_1)
   n3 = Get Base Shape Preview Actor() -> (Base Shape Preview Actor:Actor Object Reference)
   > Is Valid(InputObject=@n3)
   > Destroy Actor(self=@n3)
   n4 = Get ISM Actor()
   Is Valid(exec=@n2.then_1, InputObject=@n4)
   n5 = Length(TargetArray:Array of Instanced Static Mesh Component Object References=Get Components By Class!(self:Actor Object Reference=@n4, ComponentClass:Actor Component Class Reference=/Script/Engine.InstancedStaticMeshComponent) -> (ReturnValue:Array of Instanced Static Mesh Component Object References))
   n6 = Length(TargetArray:Array of Hierarchical Instanced Static Mesh Component Object References=Get Components By Class(self=@n4, ComponentClass=/Script/Engine.HierarchicalInstancedStaticMeshComponent) -> (ReturnValue:Array of Hierarchical Instanced Static Mesh Component Object References))
   > Branch(Condition=AND Boolean(A="integer <= integer"(A=@n5), B="integer <= integer"(A=@n6)))
   > Destroy Actor(self=Get ISM Actor())
   n7 = Event OnScriptTick()
   > Set CurTime(CurTime=float + float(A=@n7.DeltaTime, B=Get CurTime()))
   > n8 = TimedClearUserWarning!(
    execute
    ClearAfterTime:double=Get Clear Warning After Time() -> (Clear Warning After Time:double)
    CurTime:double=Get CurTime()
    LastClearTime:double=Get Last Clear Time() -> (Last Clear Time:double)
    Scriptable Tool Obj Ref:Scriptable Interactive Tool Object Reference=Self-Reference()
   ) -> (then, NewLastClearTime:double)
   > Set Last Clear Time(Last Clear Time:double=@n8) -> (Output_Get:double)
   > n9 = Branch(Condition=Get HaveAlreadyInitHeader() -> (HaveAlreadyInitHeader:bool))
   Set HaveAlreadyInitHeader(execute=@n9.else, HaveAlreadyInitHeader:bool=true) -> (Output_Get:bool)
   > n10 = Sequence!(execute) -> (then_0, then_1, then_2:Exec, then_3:Exec)
   n11 = Get Header Widget()
   > n12 = AddExpandableBox(self=@n11)
   > n13 = Set Readout Box(Readout Box:BP UW Expandable Box Object Reference=@n12) -> (Output_Get:BP UW Expandable Box Object Reference)
   > n14 = SetUpBox(
    self=@n13
    Name=Get Readout Box Label!(self:BP UW Shared Header Widget Object Reference=@n11) -> (Readout Box Label:Text)
    OpenByDefault=Get ReadoutOpenOnExit!(self:BP STPS Shared Widget State Object Reference=Get Shared State Settings() -> (Shared State Settings:BP STPS Shared Widget State Object Reference)) -> (ReadoutOpenOnExit:bool)
   )
   n15 = Set Readout Line Actor(execute=@n16.then, Readout Line Actor:BP UW Readout Line Object Reference=@n16) -> (Output_Get:BP UW Readout Line Object Reference)
   n16 = SetupLine!(
    execute=@n14.then
    self:BP UW Expandable Box Object Reference=Get Readout Box() -> (Readout Box:BP UW Expandable Box Object Reference)
    CurrentLine:BP UW Readout Line Object Reference=@n15
    DynamicLine:Vertical Box Object Reference
    Label:Text=Get Actor Line Label() -> (Actor Line Label:Text)
    Value:Text=Get Actor Line Value() -> (Actor Line Value:Text)
   ) -> (then, Readout Line:BP UW Readout Line Object Reference)
   n17 = Get Header Widget()
   n18 = AddExpandableBox(execute=@n10.then_1, self=@n17)
   > n19 = Set Shortcut Box(Shortcut Box:BP UW Expandable Box Object Reference=@n18) -> (Output_Get:BP UW Expandable Box Object Reference)
   > n20 = SetUpBox(
    self=@n19
    Name=Get Shortcut Box Label!(self:BP UW Shared Header Widget Object Reference=@n17) -> (Shortcut Box Label:Text)
    OpenByDefault=Get ShortcutsOpenOnExit!(self:BP STPS Shared Widget State Object Reference=Get Shared State Settings()) -> (ShortcutsOpenOnExit:bool)
   )
   n21 = Set ShortcutLines(execute=@n22.then, ShortcutLines:Array of BP UW Readout Line Object References=@n22) -> (Output_Get:Array of BP UW Readout Line Object References)
   n22 = AddShortcuts!(
    execute=@n20.then
    self:BP UW Expandable Box Object Reference=Get Shortcut Box() -> (Shortcut Box:BP UW Expandable Box Object Reference)
    ShortcutMap:Map of Strings to Strings=Get Shortcut Map() -> (Shortcut Map:Map of Strings to Strings)
    ExistingShortcutLines:Array of BP UW Readout Line Object References=@n21
    ShortcutMapKeys:Array of Strings=Get Shortcut Map Keys() -> (Shortcut Map Keys:Array of Strings)
   ) -> (then, AllShortcutLines:Array of BP UW Readout Line Object References)
   "SetText (Text)"(
    execute=@n10.then_2
    self=Get LabelEntry!(self:BP UW Readout Line Object Reference=Get Readout Line Actor() -> (Readout Line Actor:BP UW Readout Line Object Reference)) -> (LabelEntry:Common Text Object Reference)
    InText=Get Readout Line Actor Label() -> (Readout Line Actor Label:Text)
   )
  ]}
 Add Property Sets {
  path=EventGraph.K2Node_Composite_0.Add Property Sets
  nodes [
   Inputs()
   > n1 = AddPropertySetOfType(
    PropertySetType=&/BP_STPS_ScatterObjects.BP_STPS_ScatterObjects_C
    Identifier=Settings
   )
   > n2 = Cast To BP_STPS_ScatterObjects!(execute, Object=@n1) -> (then, CastFailed, AsBP STPS Scatter Objects:BP STPS Scatter Objects Object Reference) {is=title K2Node_DynamicCast}
   > n3 = Set Settings(Settings:BP STPS Scatter Objects Object Reference=@n2) -> (Output_Get:BP STPS Scatter Objects Object Reference)
   > RestorePropertySetSettings(PropertySet=@n3, SaveKey=Get Prop Set Save Key())
   > n4 = AddPropertySetOfType(
    PropertySetType=/Game/Blueprints/Blutilities/ToolsWrkshp/ScrTools/Shared/SharedPropertySets/BP_STPS_SharedWidgetState.BP_STPS_SharedWidgetState_C
    Identifier=SharedHeaderSettings
   )
   > n5 = Cast To BP_STPS_SharedWidgetState!(execute, Object=@n4) -> (then, CastFailed, AsBP STPS Shared Widget State:BP STPS Shared Widget State Object Reference) {is=title K2Node_DynamicCast}
   > Set Shared State Settings(Shared State Settings:BP STPS Shared Widget State Object Reference=@n5) -> (Output_Get:BP STPS Shared Widget State Object Reference)
   Outputs()
  ]
 }
 Add Behaviours {
  path=EventGraph.K2Node_Composite_1.Add Behaviours
  nodes [
   Inputs()
   > Set AddObjectHotkeyDown(AddObjectHotkeyDown:bool=false) -> (Output_Get:bool)
   > AddSingleClickBehavior!(
    execute
    self:Scriptable Modular Behavior Tool Object Reference
    IfHitByClick:Delegate=Create Event()
    OnHitByClick:Delegate=Create Event()
    CaptureCheck:Delegate=Create Event()
    CapturePriority:int=100
    MouseButton:EScriptableToolMouseButton Enum=LeftButton
    bHitTestOnRelease:bool=true
   ) -> (then)
   > AddMouseHoverBehavior!(
    execute
    self:Scriptable Modular Behavior Tool Object Reference
    BeginHoverSequenceHitTest:Delegate=Create Event()
    OnBeginHover:Delegate=Create Event()
    OnUpdateHover:Delegate=Create Event()
    OnEndHover:Delegate=Create Event()
    HoverCaptureCheck:Delegate=Create Event()
    CapturePriority:int=100
   ) -> (then)
   > AddSingleKeyInputBehavior!(
    execute
    self:Scriptable Modular Behavior Tool Object Reference
    OnKeyPressed:Delegate=Create Event()
    OnKeyReleased:Delegate=Create Event()
    OnForceEndCaptureFuncIn:Delegate=Create Event()
    Key:Key Structure=Get AddObjectHotkey() -> (AddObjectHotkey:Key Structure)
    CaptureCheck:Delegate
    CapturePriority:int=0
   ) -> (then)
   > AddSingleKeyInputBehavior(
    OnKeyPressed=Create Event()
    OnKeyReleased=Create Event()
    OnForceEndCaptureFuncIn=Create Event()
    Key=Get DoScatterHotkey() -> (DoScatterHotkey:Key Structure)
    CapturePriority=0
   )
   Outputs()
  ]
 }
 Watch Properties {
  path=EventGraph.K2Node_Composite_2.Watch Properties
  nodes [
   Inputs()
   > n1 = Sequence!(execute) -> (then_0, then_1, then_2)
   > WatchProperty(PropertySet=Get Settings(), PropertyName=Objects To Scatter, OnModified=Create Event())
   n2 = For Each Loop(Exec=@n1.then_1, Array:Array of Strings=Get Base Shape Change Properties() -> (Base Shape Change Properties:Array of Strings)) -> (Array Element:String)
   > WatchProperty(PropertySet=Get Settings(), PropertyName=@n2, OnModified=Create Event())
   WatchProperty(execute=@n1.then_2, PropertySet=Get Settings(), PropertyName=Scatter Shape Color, OnModified=Create Event())
   Outputs()
  ]
 }
 Bind To Events {
  path=EventGraph.K2Node_Composite_4.Bind To Events
  nodes [
   Inputs()
   > Bind Event to Event Clear Actor Clicked(self:BP STPS Scatter Objects Object Reference=Get Settings(), Delegate=Create Event()) {is=title K2Node_AddDelegate}
   Outputs()
  ]
 }
 Setup Const Variables {
  path=EventGraph.K2Node_Composite_5.Setup Const Variables
  nodes [
   Inputs()
   > n1 = Sequence!(execute) -> (then_0, then_1)
   > n2 = GetAllLevelActors(self=Get EditorActorSubsystem())
   > n3 = Set All Actors At Start(All Actors At Start:Array of Actor Object References=@n2) -> (Output_Get:Array of Actor Object References)
   > n4 = For Each Loop(Array:Array of Actor Object References=@n3) -> (Array Element:Actor Object Reference)
   n5 = Get Class(Object=@n4)
   > Branch(Condition=Contains(SearchIn=Get Class Display Name(Class=@n5), Substring=landscape))
   > Add Unique(
    TargetArray:Array of Actor Object References=Get Actors To Ignore In Object Selection Trace() -> (Actors To Ignore In Object Selection Trace:Array of Actor Object References)
    NewItem:"Actor Object Reference (by ref)"=@n4
   )
   n6 = GetMainProject!(execute=@n1.then_1) -> (then, ReturnValue:Valkyrie Project Handle Structure)
   > n7 = Set Project Handle(Project Handle:Valkyrie Project Handle Structure=@n6) -> (Output_Get:Valkyrie Project Handle Structure)
   > n8 = Branch(Condition=IsProjectHandleValid!(ProjectHandle:"Valkyrie Project Handle Structure (by ref)"=@n7) -> (ReturnValue:bool))
   > Set Mounted Asset Path(Mounted Asset Path:String=GetProjectMountedAssetPath!(ProjectHandle:"Valkyrie Project Handle Structure (by ref)"=Get Project Handle() -> (Project Handle:Valkyrie Project Handle Structure)) -> (ReturnValue:String)) -> (Output_Get:String)
   Set Mounted Asset Path(execute=@n8.else)
   > Show Message Dialog(
    Title=Get Message No Asset Path Title() -> (Message No Asset Path Title:Text)
    Message=Get Message No Asset Path Message() -> (Message No Asset Path Message:Text)
    MessageType=Ok
    DefaultValue=Ok
    MessageCategory=Error
   )
   Outputs()
  ]
 }
 Add UserWidgets {
  path=EventGraph.K2Node_Composite_6.Add UserWidgets
  nodes [
   Inputs()
   > n1 = Sequence()
   > n2 = Construct Object from Class(Class=&BP_UW_SharedHeaderWidget_C)
   > n3 = Set Header Widget(Header Widget:BP UW Shared Header Widget Object Reference=@n2) -> (Output_Get:BP UW Shared Header Widget Object Reference)
   > SetToolPanelToolHeaderWidget(Widget=@n3)
   n4 = Construct Object from Class(
    execute=@n1.then_1
    Class=/Game/Blueprints/Blutilities/ToolsWrkshp/ScrTools/ScatterObjects/SupportFiles/BP_UW_ScatterObjects_Overlay.BP_UW_ScatterObjects_Overlay_C
   ) -> (ReturnValue:BP UW Scatter Objects Overlay Object Reference)
   > n5 = Set Overlay Widget(Overlay Widget:BP UW Scatter Objects Overlay Object Reference=@n4) -> (Output_Get:BP UW Scatter Objects Overlay Object Reference)
   > SetOverlayWidget!(execute, self:Scriptable Interactive Tool Object Reference, Widget:User Widget Object Reference=@n5, bMakeDraggable:bool=false) -> (then)
   > Set Dropper Image(Dropper Image:Image Object Reference=Get DropperImage!(self:BP UW Scatter Objects Overlay Object Reference=Get Overlay Widget() -> (Overlay Widget:BP UW Scatter Objects Overlay Object Reference)) -> (DropperImage:Image Object Reference)) -> (Output_Get:Image Object Reference)
   > Set Overlay Dropper Appearance!(
    execute
    self
    Specified Color:Linear Color Structure=(R=1,G=1,B=1,A=1)
    InSize:Vector 2D Structure
    InPosition:Vector 2D Structure
    Denied:bool=false
   ) -> (then) {is=title K2Node_CallFunction}
   Outputs()
  ]
 }
 Find Selected ISM Actor {
  path=EventGraph.K2Node_Composite_7.Find Selected ISM Actor
  nodes [
   Inputs()
   > n1 = GetSelectedLevelActors(self=Get EditorActorSubsystem())
   > n2 = For Each Loop(Array=@n1)
   > Branch(Condition=ActorHasTag!(self:Actor Object Reference=@n2, Tag:Name=Get Scatter Tool Actor Tag() -> (Scatter Tool Actor Tag:Name=None)) -> (ReturnValue:bool))
   > Add(
    TargetArray:Array of Actor Object References=Get SelectedISMActorsOnStartup() -> (SelectedISMActorsOnStartup:Array of Actor Object References)
    NewItem:"Actor Object Reference (by ref)"=@n2
   ) {is=title K2Node_CallArrayFunction}
   n3 = Get SelectedISMActorsOnStartup()
   n4 = Branch(execute=@n2.Completed, Condition="Equal (Integer)"(A=Length(TargetArray:Array of Actor Object References=@n3), B=1))
   > Set ISM Actor(ISM Actor:Actor Object Reference="Get (a copy)"(Array:Array of Actor Object References=Get SelectedISMActorsOnStartup()) -> (Output:Actor Object Reference)) -> (Output_Get:Actor Object Reference)
   Set ISM Actor(execute=@n4.else)
   Outputs()
  ]
 }
 Init Tool State {
  path=EventGraph.K2Node_Composite_8.Init Tool State
  nodes [
   Inputs()
   > On Base Shape Changed!(execute, self, PropertySet:Scriptable Interactive Tool Property Set Object Reference, PropertyName:String) -> (then)
   > ClearUserMessages!(execute, self:Scriptable Interactive Tool Object Reference, bHelpMessage:bool=true, bWarningMessage:bool=true) -> (then)
   Outputs()
  ]
 }
 Find Or Add HISM Component {nodes [
   n1 = Add Component by Class!(
    execute:Exec
    self:Actor Object Reference=Get ISM Actor()
    Class:Actor Component Class Reference=&HierarchicalInstancedStaticMeshComponent
    bManualAttachment:bool
    RelativeTransform:Transform
   ) -> (then:Exec, ReturnValue:Hierarchical Instanced Static Mesh Component Object Reference) {is=title K2Node_AddComponentByClass}
   > n2 = Set CurHISM Component(CurHISM Component:Hierarchical Instanced Static Mesh Component Object Reference=@n1) -> (Output_Get:Hierarchical Instanced Static Mesh Component Object Reference)
   > n3 = SetStaticMesh(self=@n2, NewMesh=Get SM In Instance() -> (SM In Instance:Static Mesh Object Reference))
   Return Node!(
    execute=[@n3.then, @n5.then]
    HISM Component:Hierarchical Instanced Static Mesh Component Object Reference=Get CurHISM Component() -> (CurHISM Component:Hierarchical Instanced Static Mesh Component Object Reference)
   )
   Find Or Add HISM Component!() -> (then, SM In Instance:Static Mesh Object Reference) {is=title K2Node_FunctionEntry}
   > n4 = For Each Loop with Break!(
    Exec:Exec
    Array:Array of Hierarchical Instanced Static Mesh Component Object References=Get Components By Class(self=Get ISM Actor(), ComponentClass=&HierarchicalInstancedStaticMeshComponent)
    Break:Exec=@n12.then
   ) -> (LoopBody:Exec, Array Element:Hierarchical Instanced Static Mesh Component Object Reference, Array Index:int, Completed:Exec)
   n5 = Branch(execute=@n4.Completed, Condition=Get FoundExisting() -> (FoundExisting:bool))
   n6 = Gather Subobject Data For Instance(execute=@n5.else, self=Get SubobjectDataSubsystem(), Context=Get ISM Actor())
   > n7 = AddNewSubobject!(
    execute
    self:Subobject Data Subsystem Object Reference=Get SubobjectDataSubsystem()
    Params:"Add New Subobject Params Structure (by ref)"=Make Add New Subobject Params!(
     ParentHandle:Subobject Data Handle Structure="Get (a copy)"(Array:Array of Subobject Data Handle Structures=@n6, Dimension 1=1) -> (Output:Subobject Data Handle Structure)
     NewClass:Object Class Reference=&HierarchicalInstancedStaticMeshComponent
     BlueprintContext:Blueprint Object Reference
     bSkipMarkBlueprintModified:bool=False
     bConformTransformToParent:bool=false
    ) -> (AddNewSubobjectParams:Add New Subobject Params Structure) {is=title K2Node_MakeStruct}
   ) -> (then, FailReason:Text, ReturnValue:Subobject Data Handle Structure)
   n8 = Get SM In Instance()
   > n9 = RenameSubobject!(
    execute
    self:Subobject Data Subsystem Object Reference=Get SubobjectDataSubsystem()
    Handle:"Subobject Data Handle Structure (by ref)"=@n7.ReturnValue
    InNewName:"Text (by ref)"="To Text (String)"(InString=GetDisplayName(Object=@n8))
   ) -> (then, ReturnValue:bool)
   > PrintString!(execute, InString="To String (Boolean)"(InBool=@n9), bPrintToScreen, bPrintToLog, TextColor, Duration, Key) -> (then)
   > n10 = Cast To HierarchicalInstancedStaticMeshComponent!(
    execute
    Object=GetAssociatedObject!(Data:"Subobject Data Structure (by ref)"=GetData(DataHandle=@n7.ReturnValue)) -> (ReturnValue:Object Reference)
   ) -> (then, CastFailed, AsHierarchical Instanced Static Mesh Component:Hierarchical Instanced Static Mesh Component Object Reference) {is=title K2Node_DynamicCast}
   > Set CurHISM Component(CurHISM Component=@n10)
   > SetStaticMesh(self=@n10, NewMesh=@n8)
   > Return Node(HISM Component=@n10)
   Set CurHISM Component(execute=@n4.LoopBody, CurHISM Component=@n4)
   > n11 = Set FoundExisting(FoundExisting:bool="Equal (Object)"(A=Get SM In Instance(), B=Get StaticMesh(self=@n4))) -> (Output_Get:bool)
   > n12 = Branch(Condition=@n11)
  ]}
 Find Or Add ISM Component {nodes [
   Find Or Add ISM Component!() -> (then, SM In Instance:Static Mesh Object Reference) {is=title K2Node_FunctionEntry}
   > n1 = For Each Loop with Break(
    Array:Array of Instanced Static Mesh Component Object References=Get Components By Class(self=Get ISM Actor(), ComponentClass=&InstancedStaticMeshComponent) -> (ReturnValue:Array of Instanced Static Mesh Component Object References)
    Break=@n9.then
   ) -> (Array Element:Instanced Static Mesh Component Object Reference)
   n2 = Branch(execute=@n1.Completed, Condition=Get FoundExisting())
   > Return Node!(
    execute
    ISM Component:Instanced Static Mesh Component Object Reference=Get CurISM Component() -> (CurISM Component:Instanced Static Mesh Component Object Reference)
   )
   n3 = Gather Subobject Data For Instance(execute=@n2.else, self=Get SubobjectDataSubsystem(), Context=Get ISM Actor())
   > n4 = AddNewSubobject(
    self=Get SubobjectDataSubsystem()
    Params=Make Add New Subobject Params(
     ParentHandle="Get (a copy)"(Array=@n3, Dimension 1=1)
     NewClass=&InstancedStaticMeshComponent
     bSkipMarkBlueprintModified=False
     bConformTransformToParent=false
    )
   )
   n5 = Get SM In Instance()
   > RenameSubobject(self=Get SubobjectDataSubsystem(), Handle=@n4.ReturnValue, InNewName="To Text (String)"(InString=GetDisplayName(Object=@n5)))
   > n6 = Cast To InstancedStaticMeshComponent!(execute, Object=GetAssociatedObject(Data=GetData(DataHandle=@n4.ReturnValue))) -> (then, CastFailed, AsInstanced Static Mesh Component:Instanced Static Mesh Component Object Reference) {is=title K2Node_DynamicCast}
   > n7 = Set CurISM Component(CurISM Component:Instanced Static Mesh Component Object Reference=@n6) -> (Output_Get:Instanced Static Mesh Component Object Reference)
   > SetStaticMesh(self=@n7, NewMesh=Get SM In Instance())
   > Return Node(ISM Component=Get CurISM Component())
   Set CurISM Component(execute=@n1.LoopBody, CurISM Component=@n1)
   > n8 = Set FoundExisting(FoundExisting="Equal (Object)"(A=Get SM In Instance(), B=Get StaticMesh(self=@n1)))
   > n9 = Branch(Condition=@n8)
  ]}
 Find Or Make Preview Actor {nodes [
   Find Or Make Preview Actor() {is=title K2Node_FunctionEntry}
   > n1 = Is Valid(InputObject=Get Base Shape Preview Actor())
   n2 = SpawnActorFromClass!(
    execute=@n1.Is Not Valid
    ActorClass:Actor Class Reference=/Script/GeometryFramework.DynamicMeshActor
    Location:Vector="0, 0, 0"
    Rotation:Rotator
    bTransient:bool=true
   ) -> (then, ReturnValue:Dynamic Mesh Actor Object Reference)
   > Set Base Shape Preview Actor(Base Shape Preview Actor:Actor Object Reference=@n2) -> (Output_Get:Actor Object Reference)
   > n3 = CreateDynamicMaterialInstance!(
    execute
    Parent=/Game/Blueprints/Blutilities/ToolsWrkshp/ScrTools/Shared/Materials/MI_BoxSidePlaneNoDepthTest
    OptionalName=ScatterPreviewObjectDynMat
    CreationFlags=Transient
   ) -> (then, ReturnValue)
   > n4 = Set PreviewShapeDynMat(PreviewShapeDynMat:Material Instance Dynamic Object Reference=@n3) -> (Output_Get:Material Instance Dynamic Object Reference)
   > Set Base Shape Preview Look() {is=title K2Node_CallFunction}
   n5 = GetComponentByClass!(
    self:Actor Object Reference=Get Base Shape Preview Actor()
    ComponentClass:Actor Component Class Reference=/Script/GeometryFramework.DynamicMeshComponent
   ) -> (ReturnValue:Dynamic Mesh Component Object Reference)
   > SetMaterial(self=@n5, ElementIndex=0, Material=@n4)
   > Set Preview Actor DynMeshComp(Preview Actor DynMeshComp:Dynamic Mesh Component Object Reference=@n5) -> (Output_Get:Dynamic Mesh Component Object Reference)
   > SetActorLabel(self=Get Base Shape Preview Actor(), NewActorLabel=SCATTER PREVIEW, bMarkDirty=false)
  ]}
 Find Or Spawn ISM Actor {nodes [
   n1 = Find Or Spawn ISM Actor!() -> (then, Find Existing:bool=~) {is=title K2Node_FunctionEntry}
   > n2 = Branch(Condition=@n1)
   > n3 = GetAllActorsWithTag!(execute, Tag:Name=Get Scatter Tool Actor Tag() -> (Scatter Tool Actor Tag=None)) -> (then, OutActors:Array of Actor Object References)
   > n4 = Branch(Condition="integer <= integer"(A=Length(TargetArray=@n3)))
   n5 = SpawnActorFromClass(execute=[@n4.then, @n2.else], ActorClass=&Actor, Location="0, 0, 0") -> (ReturnValue:Actor Object Reference)
   > n6 = Set ISM Actor(ISM Actor=@n5)
   n7 = Get Settings()
   > n8 = Switch on E_ActorType_ScatterObjects!(
    execute
    Selection:E_ActorType_ScatterObjects Enum=Get Actor Type!(self:BP STPS Scatter Objects Object Reference=@n7) -> (Actor Type:E_ActorType_ScatterObjects Enum=NewEnumerator3)
   ) -> (NewEnumerator3:Exec, NewEnumerator1:Exec) {is=title K2Node_SwitchEnum}
   n9 = Set Actor Label Suffix(execute=@n8.NewEnumerator1, Actor Label Suffix:String=HISM) -> (Output_Get:String)
   n10 = Get Label Prefix!(self:BP STPS Scatter Objects Object Reference=@n7) -> (Label Prefix:String)
   n11 = Get Actor Label Suffix() -> (Actor Label Suffix:String)
   SetActorLabel(
    execute=[@n9.then, @n12.then]
    self=@n6
    NewActorLabel="To String (Text)"(InText=Format Text!(Format="LOCTEXT(\"{p} {s}\")", p:String=@n10, s:String=@n11) -> (Result))
    bMarkDirty=true
   )
   > Set Tags!(
    execute
    Tags:Array of Names=Make Array("[0]":Name=Get Scatter Tool Actor Tag() -> (Scatter Tool Actor Tag=None)) -> (Array:Array of Names)
    self:Actor Object Reference=@n6
   ) -> (then, Output_Get:Array of Names)
   n12 = Set Actor Label Suffix(execute=@n8.NewEnumerator3, Actor Label Suffix=ISM)
   Set ISM Actor(execute=@n4.else, ISM Actor="Get (a copy)"(Array:Array of Actor Object References=@n3) -> (Output:Actor Object Reference))
  ]}
 Has Instanceable Material {nodes [
   n1 = Has Instanceable Material!() -> (then, Static Mesh:Static Mesh Object Reference) {is=title K2Node_FunctionEntry}
   > Set MeshHasInvalidMat(MeshHasInvalidMat:bool=false) -> (Output_Get:bool)
   > n2 = For Each Loop!(
    Exec
    Array:Array of Static Material Structures=Get StaticMaterials!(self:Static Mesh Object Reference=@n1) -> (StaticMaterials:Array of Static Material Structures)
   ) -> (
    LoopBody
    Array Element_MaterialInterface:Material Interface Object Reference
    Array Element_MaterialSlotName:Name
    Array Element_UVChannelData:Mesh UVChannel Info Structure
    Array Element_OverlayMaterialInterface:Material Interface Object Reference
    Array Index
    Completed
   )
   Return Node!(execute=@n2.Completed, Material Instanceable:bool=NOT Boolean(A=Get MeshHasInvalidMat() -> (MeshHasInvalidMat:bool)))
   > n3 = GetBaseMaterial!(execute, self:Material Interface Object Reference=@n2) -> (then, ReturnValue:Material Object Reference)
   > Set MeshHasInvalidMat(MeshHasInvalidMat=NOT Boolean(A=Get bUsedWithInstancedStaticMeshes!(self:Material Object Reference=@n3) -> (bUsedWithInstancedStaticMeshes:bool)))
  ]}
 HoverCaptureCheck {nodes [
   HoverCaptureCheck!() -> (then, InputDeviceState:"Input Device State Structure (by ref)") {is=title K2Node_FunctionEntry}
   > Return Node!(execute, ReturnValue:bool=true)
  ]}
 HoverHitTest {nodes [
   HoverHitTest!() -> (then, CurrentPos:Input Device Ray Structure, Modifiers_bShiftDown:bool=~, Modifiers_bCtrlDown:bool=~, Modifiers_bAltDown:bool=~) {is=title K2Node_FunctionEntry}
   > SetFocusInViewport(self:Scriptable Interactive Tool Object Reference)
   > Return Node!(
    execute
    ReturnValue_bHit:bool=true
    ReturnValue_HitDepth:double=340282346638528859811704183484516925440
    ReturnValue_HitNormal:Vector="0,0,1"
    ReturnValue_bHasHitNormal:bool=False
    ReturnValue_HitIdentifier:int=0
    ReturnValue_HitObject:Object Reference
   )
  ]}
 IfHitByClick {nodes [
   n1 = IfHitByClick!() -> (
    then
    ClickPos_WorldRay:Ray Structure
    ClickPos_bHas2D:bool=~
    ClickPos_ScreenPosition:Vector 2D Structure
    MouseButton:EScriptableToolMouseButton Enum=LeftButton
   ) {is=title K2Node_FunctionEntry}
   n2 = GetRayStartEnd!(Ray:Ray Structure=@n1, StartDistance:double=0, EndDistance:double=0) -> (StartPoint:Vector, EndPoint:Vector)
   > n3 = Multi Trace Ignore Landscape!(execute, self, Start:Vector=@n2, End:Vector=@n2.EndPoint, ActorsToIgnore:Array of Actor Object References) -> (then, ReturnValue:bool, The Hit:Hit Result Structure)
   > Set Out Hit(Out Hit:Hit Result Structure=@n3.The Hit) -> (Output_Get:Hit Result Structure)
   > Return Node(
    ReturnValue_bHit=@n3
    ReturnValue_HitDepth=340282346638528859811704183484516925440
    ReturnValue_HitNormal="0,0,1"
    ReturnValue_bHasHitNormal=False
    ReturnValue_HitIdentifier=0
   )
   Get Actors To Ignore In Object Selection Trace()
  ]}
 Is In Project Folder {nodes [
   n1 = Is In Project Folder!() -> (then, Staticnmesh to Test:Static Mesh Object Reference) {is=title K2Node_FunctionEntry}
   > n2 = Is Valid(InputObject=@n1)
   Return Node!(execute=[@n2.Is Not Valid, @n4.then], In Project Folder:bool=Get Is In Project Folder() -> (Is In Project Folder:bool))
   > n3 = GetPathNameForLoadedAsset(LoadedAsset=@n1)
   > n4 = Set Is In Project Folder(Is In Project Folder:bool=StartsWith(SourceString=@n3, InPrefix=Get Mounted Asset Path() -> (Mounted Asset Path:String))) -> (Output_Get:bool)
  ]}
 IsDisallowedActor {nodes [
   IsDisallowedActor!() -> (then, ActorLabel:String) {is=title K2Node_FunctionEntry}
   > Set FoundDisallowed(FoundDisallowed:bool=false) -> (Output_Get:bool)
   > n1 = For Each Loop!(Exec, Array:Array of Strings=Get Always Disallowed Actor Labels() -> (Always Disallowed Actor Labels:Array of Strings)) -> (LoopBody, Array Element:String, Array Index, Completed)
   Return Node!(execute=@n1.Completed, IsDisallowed:bool=Get FoundDisallowed() -> (FoundDisallowed:bool))
   n2 = Get ActorLabel() -> (ActorLabel:String)
   > Branch(Condition="integer >= integer"(A=FindSubstring(SearchIn=@n2, Substring=@n1)))
   > Set FoundDisallowed(FoundDisallowed=true)
  ]}
 Multi Trace Ignore Landscape {nodes [
   n1 = Multi Trace Ignore Landscape!() -> (then, Start:Vector=~, End:Vector=~, ActorsToIgnore:Array of Actor Object References) {is=title K2Node_FunctionEntry}
   > n2 = Multi Line Trace By Channel!(
    execute
    Start:Vector=@n1
    End:Vector=@n1.End
    TraceChannel:ETraceTypeQuery Enum=TraceTypeQuery10
    bTraceComplex:bool=true
    ActorsToIgnore:Array of Actor Object References=@n1.ActorsToIgnore
    DrawDebugType:EDrawDebugTrace Enum
    bIgnoreSelf:bool=true
    TraceColor:Linear Color Structure=(R=1,G=0,B=0,A=1)
    TraceHitColor:Linear Color Structure=(R=0,G=1,B=0,A=1)
    DrawTime:float=5
   ) -> (then, OutHits:Array of Hit Result Structures, ReturnValue:bool)
   > n3 = Branch(Condition="integer > integer"(A=Length(TargetArray:Array of Hit Result Structures=@n2), B=1))
   > n4 = Branch(Condition=Get Collide Under Landscape() -> (Collide Under Landscape:bool))
   > n5 = For Each Loop with Break(Array:Array of Hit Result Structures=@n2, Break=@n7.else) -> (Array Element:Hit Result Structure)
   Return Node!(
    execute=[@n5.Completed, @n8.then]
    ReturnValue:bool=@n2.ReturnValue
    The Hit:Hit Result Structure=Get Cur Hit() -> (Cur Hit:Hit Result Structure)
   )
   > Set Cur Hit(Cur Hit:Hit Result Structure=@n5) -> (Output_Get:Hit Result Structure)
   n6 = Get Class(Object=BreakHitResult!(Hit:"Hit Result Structure (by ref)"=@n5) -> (
     bBlockingHit:bool
     bInitialOverlap:bool
     Time:float
     Distance:float
     Location:Vector
     ImpactPoint:Vector
     Normal:Vector
     ImpactNormal:Vector
     PhysMat:Physical Material Object Reference
     HitActor:Actor Object Reference
     HitComponent:Primitive Component Object Reference
     HitBoneName:Name
     BoneName:Name
     HitItem:int
     ElementIndex:int
     FaceIndex:int
     TraceStart:Vector
     TraceEnd:Vector
    ).HitActor)
   > n7 = Branch(Condition=Contains(SearchIn=Get Class Display Name(Class=@n6), Substring=landscape))
   Branch(
    execute=[@n4.else, @n3.else]
    Condition=Is Valid Index!(TargetArray:Array of Hit Result Structures=@n2, IndexToTest:int) -> (ReturnValue:bool)
   )
   > n8 = Set Cur Hit(Cur Hit="Get (a ref)"(Array:Array of Hit Result Structures=@n2, Dimension 1=0) -> (Output:"Hit Result Structure (by ref)"))
  ]}
 On Base Shape Changed {nodes [
   On Base Shape Changed!() -> (then, PropertySet:Scriptable Interactive Tool Property Set Object Reference, PropertyName:String) {is=title K2Node_FunctionEntry}
   > n1 = Sequence!(execute) -> (then_0, then_1, then_2:Exec, then_3:Exec, then_4:Exec, then_5:Exec, then_6:Exec)
   > Find Or Make Preview Actor() {is=title K2Node_CallFunction}
   n2 = Is Valid(exec=@n1.then_1, InputObject=Get Base Shape Mesh() -> (Base Shape Mesh:Dynamic Mesh Object Reference))
   n3 = CreateDynamicMeshPool!(execute=@n2.Is Not Valid) -> (then, ReturnValue:Dynamic Mesh Pool Object Reference)
   > n4 = RequestMesh!(execute, self:Dynamic Mesh Pool Object Reference=@n3) -> (then, ReturnValue:Dynamic Mesh Object Reference)
   > Set Base Shape Mesh(Base Shape Mesh:Dynamic Mesh Object Reference=@n4) -> (Output_Get:Dynamic Mesh Object Reference)
   Reset!(execute=@n1.then_2, self:Dynamic Mesh Object Reference=Get Base Shape Mesh()) -> (then, ReturnValue:Dynamic Mesh Object Reference)
   Set Prim Options(
    execute=@n1.then_3
    Prim Options:Geometry Script Primitive Options Structure=Make Geometry Script Primitive Options!(
     PolygroupMode:EGeometryScriptPrimitivePolygroupMode Enum=SingleGroup
     bFlipOrientation:bool=False
     UVMode:EGeometryScriptPrimitiveUVMode Enum=ScaleToFill
     MaterialID:int=0
    ) -> (GeometryScriptPrimitiveOptions:Geometry Script Primitive Options Structure) {is=title K2Node_MakeStruct}
   ) -> (Output_Get:Geometry Script Primitive Options Structure)
   > Set TargetArea(TargetArea:Vector 2D Structure=Get Scatter Shape Size!(self:BP STPS Scatter Objects Object Reference=Get Settings()) -> (Scatter Shape Size:Vector 2D Structure)) -> (Output_Get:Vector 2D Structure)
   n5 = Switch on E_BaseShapes_ScatterObjects!(
    execute=@n1.then_4
    Selection:E_BaseShapes_ScatterObjects Enum=Get Base Shape!(self:BP STPS Scatter Objects Object Reference=Get Settings()) -> (Base Shape:E_BaseShapes_ScatterObjects Enum=NewEnumerator0)
   ) -> (NewEnumerator0:Exec, NewEnumerator1:Exec, NewEnumerator3:Exec) {is=title K2Node_SwitchEnum}
   n6 = Get!() -> (TargetArea_X:double, TargetArea_Y:double)
   > AppendRectangleXY!(
    execute
    TargetMesh:Dynamic Mesh Object Reference=Get Base Shape Mesh()
    PrimitiveOptions:Geometry Script Primitive Options Structure=Get Prim Options() -> (Prim Options:Geometry Script Primitive Options Structure)
    Transform:Transform=MakeTransform(Rotation="0, 0, 0")
    DimensionX:float=@n6
    DimensionY:float=@n6.TargetArea_Y
    StepsWidth:int=0
    StepsHeight:int=0
   ) -> (then, ReturnValue:Dynamic Mesh Object Reference)
   n7 = AppendDisc!(
    execute=@n5.NewEnumerator1
    TargetMesh:Dynamic Mesh Object Reference=Get Base Shape Mesh()
    PrimitiveOptions:Geometry Script Primitive Options Structure=Get Prim Options()
    Transform:Transform=MakeTransform(Rotation="0, 0, 0")
    Radius:float=0.5
    AngleSteps:int=Get Circle Shape Steps() -> (Circle Shape Steps:int=0)
    SpokeSteps:int=0
    StartAngle:float=0
    EndAngle:float=360
    HoleRadius:float=0
   ) -> (then, ReturnValue:Dynamic Mesh Object Reference)
   n8 = Get()
   > ScaleMesh!(
    execute
    TargetMesh:Dynamic Mesh Object Reference=@n7
    Scale:Vector=MakeVector(X=@n8, Y=@n8.TargetArea_Y, Z=1)
    ScaleOrigin:Vector="0,0,0"
    bFixOrientationForNegativeScale:bool=false
   ) -> (then, ReturnValue:Dynamic Mesh Object Reference)
   AppendDisc(
    execute=@n5.NewEnumerator3
    TargetMesh=Get Base Shape Mesh()
    PrimitiveOptions=Get Prim Options()
    Transform=MakeTransform(Rotation="0, 0, 0")
    Radius=10
    AngleSteps=Get Circle Shape Steps() -> (Circle Shape Steps=0)
    SpokeSteps=0
    StartAngle=0
    EndAngle=360
    HoleRadius=0
   )
   > Apply Perlin Noise to Mesh!(
    execute
    TargetMesh:Dynamic Mesh Object Reference=Get Base Shape Mesh()
    Selection:Mesh Selection Structure
    Options_BaseLayer_Magnitude:float=2.5
    Options_BaseLayer_Frequency:float=0.25
    Options_BaseLayer_FrequencyShift:Vector="0,0,0"
    Options_BaseLayer_RandomSeed:int=RandomIntegerFromStream!(Stream:"Random Stream Structure (by ref)"=Get Cur Click Random Stream() -> (Cur Click Random Stream:Random Stream Structure), Max:int=500) -> (ReturnValue:int=0)
    Options_bApplyAlongNormal:bool=false
    Options_NormalSource:EGeometryScriptPerVertexNormalSource Enum=Computed
    Options_EmptyBehavior:EGeometryScriptEmptySelectionBehavior Enum=FullMeshSelection
   ) -> (then, ReturnValue:Dynamic Mesh Object Reference)
   n9 = Get()
   > ScaleMesh(
    TargetMesh=Get Base Shape Mesh()
    Scale=vector * vector(A=MakeVector(X=@n9, Y=@n9.TargetArea_Y), B:Vector="0.05,0.05,0.1") -> (ReturnValue=~)
    ScaleOrigin="0,0,0"
    bFixOrientationForNegativeScale=true
   )
   SetDynamicMesh!(
    execute=@n1.then_5
    self:Dynamic Mesh Component Object Reference=Get Preview Actor DynMeshComp() -> (Preview Actor DynMeshComp:Dynamic Mesh Component Object Reference)
    NewMesh:Dynamic Mesh Object Reference=Get Base Shape Mesh()
   ) -> (then)
   Set Base Shape Preview Look(execute=@n1.then_6)
  ]}
 On Base Shape Look Changed {nodes [
   On Base Shape Look Changed!() -> (then, PropertySet:Scriptable Interactive Tool Property Set Object Reference, PropertyName:String) {is=title K2Node_FunctionEntry}
   > Set Base Shape Preview Look()
  ]}
 On Clear Actor Clicked {nodes [
   On Clear Actor Clicked() {is=title K2Node_FunctionEntry}
   > Is Valid(InputObject=Get ISM Actor())
   > BeginTransaction(Context=Get Undo Clear Context(), Description=Get Undo Clear Description() -> (Undo Clear Description:Text))
   > TransactObject(Object=Get ISM Actor())
   > n1 = For Each Loop(Array:Array of Hierarchical Instanced Static Mesh Component Object References=Get Components By Class(self=Get ISM Actor(), ComponentClass=&HierarchicalInstancedStaticMeshComponent) -> (ReturnValue:Array of Hierarchical Instanced Static Mesh Component Object References)) -> (Array Element:Hierarchical Instanced Static Mesh Component Object Reference)
   n2 = For Each Loop(
    Exec=@n1.Completed
    Array:Array of Instanced Static Mesh Component Object References=Get Components By Class(self=Get ISM Actor(), ComponentClass=&InstancedStaticMeshComponent) -> (ReturnValue:Array of Instanced Static Mesh Component Object References)
   ) -> (Array Element:Instanced Static Mesh Component Object Reference)
   EndTransaction(execute=@n2.Completed)
   ClearInstances(execute=@n2.LoopBody, self:Instanced Static Mesh Component Object Reference=@n2)
   ClearInstances(execute=@n1.LoopBody, self=@n1)
  ]}
 On End Hover {nodes [
   On End Hover() {is=title K2Node_FunctionEntry}
   > Set Overlay Dropper Appearance(Denied=false)
  ]}
 On Objects To Scatter Changed {nodes [
   On Objects To Scatter Changed!() -> (then, PropertySet:Scriptable Interactive Tool Property Set Object Reference, PropertyName:String) {is=title K2Node_FunctionEntry}
   > Set Assets To Scatter(Assets To Scatter:Array of S Selection Choices Scatter Objects Structures=Get Objects To Scatter!(self:BP STPS Scatter Objects Object Reference=Get Settings()) -> (Objects To Scatter:Array of S Selection Choices Scatter Objects Structures)) -> (Output_Get:Array of S Selection Choices Scatter Objects Structures)
   > n1 = Sequence!(execute) -> (then_0, then_1)
   > n2 = For Each Loop(Array:Array of S Selection Choices Scatter Objects Structures=Get Assets To Scatter() -> (Assets To Scatter:Array of S Selection Choices Scatter Objects Structures)) -> (Array Element:S Selection Choices Scatter Objects Structure)
   n3 = For Each Loop(Exec=@n2.Completed, Array=Get Assets not in project() -> (Assets not in project:Array of S Selection Choices Scatter Objects Structures))
   > Remove Item!(
    execute
    TargetArray:Array of S Selection Choices Scatter Objects Structures=Get Assets To Scatter()
    Item:"S Selection Choices Scatter Objects Structure (by ref)"=@n3
   ) -> (then, ReturnValue:bool)
   > Remove Item(TargetArray=Get Objects To Scatter(self=Get Settings()), Item=@n3)
   Is Valid(exec=@n2.LoopBody, InputObject=Break S Selection Choices Scatter Objects(S_SelectionChoices_ScatterObjects=@n2))
   n4 = Break S Selection Choices Scatter Objects(S_SelectionChoices_ScatterObjects=@n2)
   > n5 = Is In Project Folder!(execute, self, Staticnmesh to Test:Static Mesh Object Reference=@n4) -> (then, In Project Folder:bool) {is=title K2Node_CallFunction}
   > n6 = Branch(Condition=@n5)
   DisplayUserWarningMessage(
    execute=@n6.else
    Message=Format Text!(Format=Get Disallowed Project Location Message() -> (Disallowed Project Location Message:Text), 0:Static Mesh Object Reference=@n4) -> (Result)
   )
   > Set Last Clear Time(Last Clear Time=float + float(A=Get CurTime(), B=0.1))
   > Add(
    TargetArray:Array of S Selection Choices Scatter Objects Structures=Get Assets not in project()
    NewItem:"S Selection Choices Scatter Objects Structure (by ref)"=@n2
   )
   n7 = Check Material Validity!(execute=@n1.then_1, self, Assets To Scatter:Array of S Selection Choices Scatter Objects Structures=Get Assets To Scatter()) -> (
    then
    Assets Instanceable:Array of S Selection Choices Scatter Objects Structures
    Assets Non Instanceable:Array of S Selection Choices Scatter Objects Structures
   ) {is=title K2Node_CallFunction}
   > Branch(Condition="To Boolean (Integer)"!(InInt:int=Length(TargetArray:Array of S Selection Choices Scatter Objects Structures=@n7.Assets Non Instanceable)) -> (ReturnValue:bool))
   > Set Disallowed Material Message(Disallowed Material Message:Text) -> (Output_Get:Text)
   > n8 = For Each Loop(Array=@n7.Assets Non Instanceable)
   n9 = For Each Loop(Exec=@n8.Completed, Array=@n7.Assets Non Instanceable)
   DisplayUserWarningMessage(execute=@n9.Completed, Message=Get Disallowed Material Message() -> (Disallowed Material Message:Text))
   > Set Last Clear Time(Last Clear Time=float + float(A=Get CurTime(), B=0.1))
   Remove Item(execute=@n9.LoopBody, TargetArray=Get Objects To Scatter(self=Get Settings()), Item=@n9)
   n10 = GetDisplayName(Object=Break S Selection Choices Scatter Objects(S_SelectionChoices_ScatterObjects=@n8))
   n11 = Get Disallowed Material Base Message() -> (Disallowed Material Base Message:Text)
   Set Disallowed Material Message(
    execute=@n8.LoopBody
    Disallowed Material Message=Format Text!(Format="LOCTEXT(\"{m}\\r\\n{dm}\")", m:Text=Format Text!(Format=@n11, 0:String=@n10) -> (Result), dm:Text=Get Disallowed Material Message()) -> (Result)
   )
   Return Node!(execute)
  ]}
 OnAddObjectKeyPressed {nodes [
   OnAddObjectKeyPressed!() -> (then, Key:Key Structure=None, Modifiers:Scriptable Tool Modifier States Structure) {is=title K2Node_FunctionEntry}
   > Set AddObjectHotkeyDown(AddObjectHotkeyDown=true)
  ]}
 OnAddObjectKeyReleased {nodes [
   OnAddObjectKeyReleased!() -> (then, Key:Key Structure=None, Modifiers:Scriptable Tool Modifier States Structure) {is=title K2Node_FunctionEntry}
   > Set AddObjectHotkeyDown(AddObjectHotkeyDown=false)
   > Set Overlay Dropper Appearance(Denied=false)
  ]}
 OnBeginHover {nodes [
   n1 = OnBeginHover!() -> (
    then
    CurrentPos_WorldRay:Ray Structure
    CurrentPos_bHas2D:bool=~
    CurrentPos_ScreenPosition:Vector 2D Structure
    Modifiers_bShiftDown:bool=~
    Modifiers_bCtrlDown:bool=~
    Modifiers_bAltDown:bool=~
   ) {is=title K2Node_FunctionEntry}
   n2 = GetRayStartEnd(Ray=@n1, StartDistance=0, EndDistance=0)
   n3 = Line Trace By Channel!(
    execute
    Start:Vector=@n2
    End:Vector=@n2.EndPoint
    TraceChannel:ETraceTypeQuery Enum
    bTraceComplex:bool=true
    ActorsToIgnore:Array of Actor Object References
    DrawDebugType:EDrawDebugTrace Enum
    bIgnoreSelf:bool
    TraceColor:Linear Color Structure
    TraceHitColor:Linear Color Structure
    DrawTime:float
   ) -> (then, OutHit:Hit Result Structure, ReturnValue:bool)
   > Set Hover Out Hit(Hover Out Hit:Hit Result Structure=@n3) -> (Output_Get:Hit Result Structure)
   Return Node(execute=@n5.then)
   Get Actors To Ignore In Object Selection Trace()
   Update Info(execute=[@n9.else, @n10.then])
   n4 = BreakHitResult(Hit=Get Hover Out Hit() -> (Hover Out Hit:Hit Result Structure))
   > n5 = Set Actor Location!(
    execute
    self:Actor Object Reference=Get Base Shape Preview Actor()
    NewLocation:Vector=@n4.Location
    bSweep:bool=false
    bTeleport:bool=false
   ) -> (then, SweepHitResult:Hit Result Structure, ReturnValue:bool) {is=title K2Node_CallFunction}
   n6 = Multi Trace Ignore Landscape(execute=@n1.then, Start=@n2, End=@n2.EndPoint) {is=title K2Node_CallFunction}
   > Set Hover Out Hit(Hover Out Hit=@n6.The Hit)
   > Is Valid(InputObject=@n4.HitActor)
   > n7 = Set CurrentHoverActorLabel(CurrentHoverActorLabel:String=GetActorLabel(self=@n4.HitActor, bCreateIfNone=false)) -> (Output_Get:String)
   > n8 = IsDisallowedActor!(execute, self, ActorLabel:String=@n7) -> (then, IsDisallowed:bool) {is=title K2Node_CallFunction}
   > n9 = Branch(Condition=@n8)
   > n10 = Set CurrentHoverActorLabel()
  ]}
 OnDoScatterHotkeyPressed {nodes [
   OnDoScatterHotkeyPressed!() -> (then, Key:Key Structure=None, Modifiers:Scriptable Tool Modifier States Structure) {is=title K2Node_FunctionEntry}
   > Set Do Scatter Hotkey Down(Do Scatter Hotkey Down:bool=true) -> (Output_Get:bool)
  ]}
 OnDoScatterHotkeyReleased {nodes [
   OnDoScatterHotkeyReleased!() -> (then, Key:Key Structure=None, Modifiers:Scriptable Tool Modifier States Structure) {is=title K2Node_FunctionEntry}
   > Set Do Scatter Hotkey Down(Do Scatter Hotkey Down=false)
  ]}
 OnForceEndCapture Add Objects Hotkey {nodes [
   OnForceEndCapture Add Objects Hotkey() {is=title K2Node_FunctionEntry}
   > Set AddObjectHotkeyDown(AddObjectHotkeyDown=false)
  ]}
 OnForceEndCaptureDoScatterHotkey {nodes [
   OnForceEndCaptureDoScatterHotkey() {is=title K2Node_FunctionEntry}
   > Set Do Scatter Hotkey Down(Do Scatter Hotkey Down=false)
  ]}
 OnHitByClick {nodes [
   OnHitByClick!() -> (
    then
    ClickPos:Input Device Ray Structure
    Modifiers:Scriptable Tool Modifier States Structure
    MouseButton:EScriptableToolMouseButton Enum=LeftButton
   ) {is=title K2Node_FunctionEntry}
   > n1 = Branch(Condition=Get AddObjectHotkeyDown() -> (AddObjectHotkeyDown:bool))
   > n2 = Set HitActor(HitActor:Actor Object Reference=BreakHitResult(Hit=Get Out Hit() -> (Out Hit:Hit Result Structure)).HitActor) -> (Output_Get:Actor Object Reference)
   > n3 = Is Valid(InputObject=@n2)
   Return Node(execute=@n3.Is Not Valid)
   > n4 = Set HitActorLabel(HitActorLabel:String=GetActorLabel(self=Get HitActor() -> (HitActor:Actor Object Reference), bCreateIfNone=false)) -> (Output_Get:String)
   > n5 = IsDisallowedActor(ActorLabel=Get HitActorLabel() -> (HitActorLabel:String))
   > n6 = Branch(Condition=@n5)
   > DisplayUserWarningMessage(Message=Format Text(Format=Get Disallowed Actor Message() -> (Disallowed Actor Message:Text), 0=@n4))
   > Set Last Clear Time(Last Clear Time=float + float(A=Get CurTime(), B=0.1))
   > Return Node()
   n7 = Get StaticMesh(self=GetComponentByClass(self=Get HitActor(), ComponentClass=&StaticMeshComponent) -> (ReturnValue:Static Mesh Component Object Reference))
   n8 = Is In Project Folder(execute=@n6.else, Staticnmesh to Test=@n7)
   > n9 = Branch(Condition=@n8)
   n10 = Make S Selection Choices Scatter Objects!(Mesh_15_#1:Static Mesh Object Reference=@n7, Weight_5_#2:int=100) -> (S_SelectionChoices_ScatterObjects:S Selection Choices Scatter Objects Structure) {is=title K2Node_MakeStruct}
   n11 = Get Objects To Scatter(self=Get Settings())
   > n12 = Branch(Condition="integer >= integer"(A=Find Item!(
      TargetArray:Array of S Selection Choices Scatter Objects Structures=@n11
      ItemToFind:"S Selection Choices Scatter Objects Structure (by ref)"=@n10
     ) -> (ReturnValue)))
   n13 = Break S Selection Choices Scatter Objects(S_SelectionChoices_ScatterObjects=@n10)
   n14 = GetDisplayName(Object=@n13)
   > DisplayUserWarningMessage(Message=Format Text(Format=Get Already In List Message() -> (Already In List Message:Text), 0=@n14))
   > Set Last Clear Time(Last Clear Time=float + float(A=Get CurTime(), B=0.1))
   > Return Node()
   n15 = Has Instanceable Material(execute=@n12.else, Static Mesh=@n13) {is=title K2Node_CallFunction}
   > n16 = Branch(Condition=@n15)
   > Add Unique(
    TargetArray:Array of S Selection Choices Scatter Objects Structures=Get Objects To Scatter(self=Get Settings())
    NewItem:"S Selection Choices Scatter Objects Structure (by ref)"=@n10
   )
   > DisplayUserHelpMessage!(
    execute
    self:Scriptable Interactive Tool Object Reference
    Message:Text=Format Text(Format=Get Added To List Message() -> (Added To List Message:Text), 0:Static Mesh Object Reference=@n13)
   ) -> (then)
   DisplayUserWarningMessage(execute=@n16.else, Message=Format Text(Format=Get Disallowed Material Base Message(), 0:String=@n14))
   > Set Last Clear Time(Last Clear Time=float + float(A=Get CurTime(), B=0.1))
   DisplayUserWarningMessage(execute=@n9.else, Message=Format Text(Format=Get Disallowed Project Location Message(), 0:Static Mesh Object Reference=@n7))
   > Set Last Clear Time(Last Clear Time=float + float(A=Get CurTime(), B=0.1))
   > Return Node()
   Branch(execute=@n1.else, Condition=Get Do Scatter Hotkey Down() -> (Do Scatter Hotkey Down:bool))
   > Clear(TargetArray:Array of S Selection Choices Scatter Objects Structures=Get Assets To Scatter())
   > n17 = For Each Loop(Array=Get Objects To Scatter(self=Get Settings()))
   n18 = Scatter!(execute=@n17.Completed, self) -> (
    then
    Instanceable Samples:Array of S Inst Data Scatter Objects Structures
    Non Instanceable Samples:Array of S Inst Data Scatter Objects Structures
   )
   n19 = Length(TargetArray:Array of S Inst Data Scatter Objects Structures=@n18.Non Instanceable Samples)
   n20 = Length(TargetArray=@n18)
   > Branch(Condition="integer >= integer"(A=int + int(A=@n20, B=@n19), B=1))
   > n21 = Set Scattered Assets Inst(Scattered Assets Inst:Array of S Inst Data Scatter Objects Structures=@n18) -> (Output_Get:Array of S Inst Data Scatter Objects Structures)
   > Set Combine Inst And Non(Combine Inst And Non:Array of S Inst Data Scatter Objects Structures=@n21) -> (Output_Get:Array of S Inst Data Scatter Objects Structures)
   > Set Scattered Assets Non Inst(Scattered Assets Non Inst:Array of S Inst Data Scatter Objects Structures=@n18.Non Instanceable Samples) -> (Output_Get:Array of S Inst Data Scatter Objects Structures)
   > Append Array!(
    execute
    TargetArray:Array of S Inst Data Scatter Objects Structures=Get Combine Inst And Non() -> (Combine Inst And Non:Array of S Inst Data Scatter Objects Structures)
    SourceArray:Array of S Inst Data Scatter Objects Structures=Get Scattered Assets Non Inst() -> (Scattered Assets Non Inst:Array of S Inst Data Scatter Objects Structures)
   ) -> (then)
   > On Base Shape Changed() {is=title K2Node_CallFunction}
   > n22 = Switch on E_ActorType_ScatterObjects(Selection=Get Actor Type(self=Get Settings()) -> (Actor Type=NewEnumerator3))
   n23 = Is Valid(exec=@n22.NewEnumerator1, InputObject=Get ISM Actor())
   n24 = Find Or Spawn ISM Actor!(execute=@n23.Is Not Valid, self, Find Existing:bool=false) -> (then) {is=title K2Node_CallFunction}
   n25 = Create Scattered HISM!(execute=[@n24.then, @n23.Is Valid], self, Scattered Assets:Array of S Inst Data Scatter Objects Structures=Get Combine Inst And Non()) -> (then) {is=title K2Node_CallFunction}
   n26 = Get Scattered Assets Non Inst()
   Branch(execute=[@n25.then, @n29.then], Condition="integer > integer"(A=Length(TargetArray=@n26)))
   > DisplayUserWarningMessage(Message=Get Warn Non Inst Message() -> (Warn Non Inst Message:Text))
   > Set Last Clear Time(Last Clear Time=float + float(A=Get CurTime(), B=0.1))
   n27 = Is Valid(exec=@n22.NewEnumerator3, InputObject=Get ISM Actor())
   n28 = Find Or Spawn ISM Actor(execute=@n27.Is Not Valid, Find Existing=false)
   n29 = Create Scattered ISM!(execute=[@n28.then, @n27.Is Valid], self, Scattered Assets:Array of S Inst Data Scatter Objects Structures=Get Combine Inst And Non()) -> (then) {is=title K2Node_CallFunction}
   Branch(execute=@n17.LoopBody, Condition=IsValid(Object=Break S Selection Choices Scatter Objects(S_SelectionChoices_ScatterObjects=@n17)))
   > Add Unique(TargetArray=Get Assets To Scatter(), NewItem=@n17)
  ]}
 OnHoverUpdated {nodes [
   n1 = OnHoverUpdated!() -> (
    then
    CurrentPos_WorldRay:Ray Structure
    CurrentPos_bHas2D:bool=~
    CurrentPos_ScreenPosition:Vector 2D Structure
    Modifiers_bShiftDown:bool=~
    Modifiers_bCtrlDown:bool=~
    Modifiers_bAltDown:bool=~
   ) {is=title K2Node_FunctionEntry}
   > Set Cur Screen Pos(Cur Screen Pos:Vector 2D Structure=@n1.CurrentPos_ScreenPosition) -> (Output_Get:Vector 2D Structure)
   n2 = Get Do Scatter Hotkey Down()
   > SetIsTemporarilyHiddenInEditor!(
    execute
    self:Actor Object Reference=Get Base Shape Preview Actor()
    bIsHidden:bool=OR Boolean(A=Get AddObjectHotkeyDown(), B=NOT Boolean(A=@n2))
   ) -> (then)
   n3 = GetRayStartEnd(Ray=@n1, StartDistance=0, EndDistance=0)
   > n4 = Multi Trace Ignore Landscape(Start=@n3, End=@n3.EndPoint)
   > Set Hover Out Hit(Hover Out Hit=@n4.The Hit)
   n5 = BreakHitResult(Hit=Get Hover Out Hit())
   > Is Valid(InputObject=@n5.HitActor)
   > n6 = Set CurrentHoverActorLabel(CurrentHoverActorLabel=GetActorLabel(self=@n5.HitActor, bCreateIfNone=false))
   > n7 = IsDisallowedActor(ActorLabel=@n6)
   > n8 = Sequence!(execute) -> (then_0, then_1, then_2:Exec, then_3:Exec)
   > Set Actor Location(self=Get Base Shape Preview Actor(), NewLocation=@n5.Location, bSweep=false, bTeleport=false)
   Update Info(execute=@n8.then_1)
   Branch(execute=@n8.then_2, Condition=Get AddObjectHotkeyDown())
   > n9 = Cast To StaticMeshComponent(Object=@n5.HitComponent)
   n10 = Get StaticMesh(self=@n9)
   > n11 = Has Instanceable Material(Static Mesh=@n10)
   > n12 = Is In Project Folder(Staticnmesh to Test=@n10)
   > n13 = Branch(Condition=AND Boolean!(A=@n11, B=NOT Boolean(A=@n7), C:bool=@n12) -> (ReturnValue))
   > Set Overlay Dropper Appearance(
    Specified Color=(R=1,G=1,B=1,A=1)
    InSize=Get Dropper Size() -> (Dropper Size:Vector 2D Structure)
    InPosition=Get Cur Screen Pos() -> (Cur Screen Pos:Vector 2D Structure)
    Denied=false
   )
   Set Overlay Dropper Appearance(
    execute=[@n9.CastFailed, @n13.else]
    Specified Color=(R=1,G=1,B=1,A=1)
    InSize=Get Dropper Size()
    InPosition=Get Cur Screen Pos()
    Denied=true
   )
   Return Node!(execute=@n8.then_3, ReturnValue:bool=true)
  ]}
 Scatter {nodes [
   Scatter() {is=title K2Node_FunctionEntry}
   > Clear(TargetArray=Get Instanceable Meshes To Scatter() -> (Instanceable Meshes To Scatter:Array of S Selection Choices Scatter Objects Structures))
   > n1 = GetActorsTempHidden!(execute) -> (then, ActorTempHiddenMap:Map of Actor Object References to Booleans, OnlyTempHidden:Array of Actor Object References)
   > Set ActorsHiddenInEditor(ActorsHiddenInEditor:Array of Actor Object References=@n1.OnlyTempHidden) -> (Output_Get:Array of Actor Object References)
   > n2 = Sequence!(execute) -> (then_0, then_1, then_2, then_3, then_4:Exec, then_5:Exec, then_6:Exec)
   TranslateMesh!(
    execute=@n2.then_2
    TargetMesh:Dynamic Mesh Object Reference=Get Base Shape Mesh()
    Translation:Vector=BreakHitResult(Hit=Get Out Hit()).Location
   ) -> (then, ReturnValue:Dynamic Mesh Object Reference)
   n3 = Get Assets To Scatter()
   n4 = Branch(execute=@n2.then_3, Condition="integer > integer"(A=Length(TargetArray:Array of S Selection Choices Scatter Objects Structures=@n3)))
   > Set Num Samples This Click(Num Samples This Click:int=0) -> (Output_Get:int=0)
   > Set MaxRadius(MaxRadius:double=0) -> (Output_Get:double)
   > K2Node_Composite_2 = "Init Max Radius and Num Samples "!(Exec:Exec) -> (Completed:Exec) {is=title K2Node_Composite}
   > Set CurClickRandInt(CurClickRandInt:int=RandomIntegerFromStream(Stream=Get Cur Click Random Stream(), Max=50000) -> (ReturnValue=0)) -> (Output_Get:int=0)
   > n5 = Branch(Condition=Get Allow Overlap!(self:BP STPS Scatter Objects Object Reference=Get Settings()) -> (Allow Overlap:bool))
   > n6 = ComputeUniformRandomPointSampling!(
    execute
    TargetMesh:Dynamic Mesh Object Reference=Get Base Shape Mesh()
    NumSamples:int=Get Num Samples This Click() -> (Num Samples This Click:int=0)
    RandomSeed:int=Get CurClickRandInt() -> (CurClickRandInt:int=0)
   ) -> (then, Samples:Array of Transforms, TriangleIDs:Index List Structure, ReturnValue:Dynamic Mesh Object Reference)
   > n7 = Set Raw Samples(Raw Samples:Array of Transforms=@n6) -> (Output_Get:Array of Transforms)
   Set Num Samples(
    execute=[@n7.then, @n24.then]
    Num Samples:int=Length(TargetArray:Array of Transforms=Get Raw Samples() -> (Raw Samples:Array of Transforms))
   ) -> (Output_Get:int=0)
   > n8 = For Each Loop(Array:Array of Transforms=Get Raw Samples()) -> (Array Element:Transform)
   > Set Current Sample Point Index(Current Sample Point Index:int=@n8.Array Index) -> (Output_Get:int=0)
   n9 = BreakVector(InVec=BreakTransform(InTransform=@n8))
   > n10 = Line Trace By Channel(
    Start=MakeVector(X=@n9, Y=@n9.Y, Z=1000000)
    End=MakeVector(X=@n9, Y=@n9.Y, Z=-100000000)
    TraceChannel=TraceTypeQuery10
    bTraceComplex=true
    ActorsToIgnore=Get ActorsToIgnoreInTrace() -> (ActorsToIgnoreInTrace:Array of Actor Object References)
   )
   > n11 = Set Scatter Hit(Scatter Hit:Hit Result Structure=@n10) -> (Output_Get:Hit Result Structure)
   > Branch()
   n12 = "Get (a ref)"(
    Array:Array of S Selection Choices Scatter Objects Structures=Get Asset Sample Counts() -> (Asset Sample Counts:Array of S Selection Choices Scatter Objects Structures)
    Dimension 1=Get Current Sample Point Index() -> (Current Sample Point Index:int=0)
   ) -> (Output:"S Selection Choices Scatter Objects Structure (by ref)")
   n13 = BreakVector(InVec=BreakHitResult(Hit=@n11).Location)
   n14 = Get Settings()
   n15 = MakeVector(
    X=@n13
    Y=@n13.Y
    Z=float + float(A=@n13.Z, B=Get Vertical Offset!(self:BP STPS Scatter Objects Object Reference=@n14) -> (Vertical Offset:double))
   )
   n16 = int + int(A=Get ClickSeed() -> (ClickSeed:int=0), B=Get Current Sample Point Index() -> (Current Sample Point Index=0))
   n17 = Get Settings()
   n18 = Get!(self:BP STPS Scatter Objects Object Reference=@n17) -> (Scale Range_X:double, Scale Range_Y:double)
   n19 = "To Vector (Float)"!(InDouble:double=RandomFloatInRangeFromStream!(
     Stream:"Random Stream Structure (by ref)"=MakeRandomStream!(InitialSeed:int=@n16) -> (ReturnValue:Random Stream Structure)
     Min:float=@n18
     Max:float=@n18.Scale Range_Y
    ) -> (ReturnValue:float)) -> (ReturnValue:Vector)
   n20 = int + int(A=@n16, B=15)
   n21 = BreakVector2D(InVec=Get Rotation Range!(self:BP STPS Scatter Objects Object Reference=@n17) -> (Rotation Range:Vector 2D Structure))
   n22 = MakeRotator(Yaw=RandomFloatInRangeFromStream(Stream=MakeRandomStream(InitialSeed=@n20), Min=@n21, Max=@n21.Y))
   > Add(
    TargetArray:Array of S Inst Data Scatter Objects Structures=Get Instanceable Samples() -> (Instanceable Samples:Array of S Inst Data Scatter Objects Structures)
    NewItem:"S Inst Data Scatter Objects Structure (by ref)"=Make S Inst Data Scatter Objects!(
     MeshToScatter_8_#1:Static Mesh Object Reference=Break S Selection Choices Scatter Objects(S_SelectionChoices_ScatterObjects=@n12)
     InstTransform_14_#3:Transform=MakeTransform(Location=@n15, Rotation=@n22, Scale=@n19)
    ) -> (S_InstData_ScatterObjects:S Inst Data Scatter Objects Structure) {is=title K2Node_MakeStruct}
   )
   n23 = ComputePointSampling!(
    execute=@n5.else
    TargetMesh:Dynamic Mesh Object Reference=Get Base Shape Mesh()
    Options_SamplingRadius:float=Get MaxRadius() -> (MaxRadius:double)
    Options_MaxNumSamples:int=Get Num Samples This Click() -> (Num Samples This Click=0)
    Options_RandomSeed:int=Get CurClickRandInt() -> (CurClickRandInt=0)
    Options_SubSampleDensity:double=10
    Options_SamplingMethodVersion:int=-1
   ) -> (then, Samples:Array of Transforms, TriangleIDs:Index List Structure, ReturnValue:Dynamic Mesh Object Reference)
   > n24 = Set Raw Samples(Raw Samples=@n23)
   Return Node!(
    execute=@n4.else
    Instanceable Samples:Array of S Inst Data Scatter Objects Structures
    Non Instanceable Samples:Array of S Inst Data Scatter Objects Structures
   )
   n25 = Get Out Hit()
   TranslateMesh(execute=@n2.then_4, TargetMesh=Get Base Shape Mesh(), Translation=vector - vector(A=MakeVector(Y=0), B=BreakHitResult(Hit=@n25).Location))
   Set ClickSeed(execute=@n2.then_5, ClickSeed:int=int + int(A=Get ClickSeed() -> (ClickSeed=0))) -> (Output_Get:int=0)
   Set Cur Click Random Stream(execute=@n2.then_1, Cur Click Random Stream:Random Stream Structure=MakeRandomStream(InitialSeed=Get ClickSeed() -> (ClickSeed=0))) -> (Output_Get:Random Stream Structure)
   > n26 = Set ActorsToIgnoreInTrace(ActorsToIgnoreInTrace:Array of Actor Object References=Make Array("[0]":Actor Object Reference=Get ISM Actor()) -> (Array:Array of Actor Object References)) -> (Output_Get:Array of Actor Object References)
   > Append Array(
    TargetArray:Array of Actor Object References=@n26
    SourceArray:Array of Actor Object References=Get ActorsHiddenInEditor() -> (ActorsHiddenInEditor:Array of Actor Object References)
   )
   n27 = Get Assets To Scatter()
   n28 = Branch(execute=@n2.then_0, Condition="To Boolean (Integer)"(InInt=Length(TargetArray:Array of S Selection Choices Scatter Objects Structures=@n27)))
   > n29 = Check Material Validity(Assets To Scatter=Get Assets To Scatter())
   > Set Instanceable Meshes To Scatter(Instanceable Meshes To Scatter:Array of S Selection Choices Scatter Objects Structures=@n29) -> (Output_Get:Array of S Selection Choices Scatter Objects Structures)
   > Set Non Instanceable Meshes To Scatter(Non Instanceable Meshes To Scatter:Array of S Selection Choices Scatter Objects Structures=@n29.Assets Non Instanceable) -> (Output_Get:Array of S Selection Choices Scatter Objects Structures)
   DisplayUserWarningMessage(execute=@n28.else, Message="LOCTEXT(\"No Objects To Scatter in list!\")")
   > Set Last Clear Time(Last Clear Time=Get CurTime())
   > Return Node()
   Return Node(
    execute=@n2.then_6
    Instanceable Samples=Get Instanceable Samples()
    Non Instanceable Samples=Get Non Instanceable Samples() -> (Non Instanceable Samples:Array of S Inst Data Scatter Objects Structures)
   )
  ]}
 Scatter Actors {nodes [
   n1 = Scatter Actors!() -> (then, Scattered Assets:Array of S Inst Data Scatter Objects Structures) {is=title K2Node_FunctionEntry}
   > n2 = For Each Loop(Array:Array of S Inst Data Scatter Objects Structures=@n1) -> (Array Element:S Inst Data Scatter Objects Structure)
   n3 = Break S Inst Data Scatter Objects(S_InstData_ScatterObjects=@n2)
   n4 = BreakTransform(InTransform=@n3.InstTransform_14_#3)
   > n5 = SpawnActorFromClass(ActorClass=&StaticMeshActor, Location=@n4, Rotation=@n4.Rotation) -> (ReturnValue:Static Mesh Actor Object Reference)
   > SetActorScale3D!(execute, self:Actor Object Reference=@n5, NewScale3D:Vector=@n4.Scale) -> (then)
   > SetStaticMesh(self=Get StaticMeshComponent(self=@n5), NewMesh=@n3)
   n6 = GetDisplayName(Object=@n3)
   n7 = Get Label Prefix(self=Get Settings())
   > SetActorLabel(
    self=@n5
    NewActorLabel="To String (Text)"(InText=Format Text!(Format="LOCTEXT(\"{p} {sm}\")", p:String=@n7, sm:String=@n6) -> (Result))
    bMarkDirty=true
   )
   > Set Tags(Tags=Make Array("[0]":Name=Get Scatter Tool Actor Tag() -> (Scatter Tool Actor Tag=None)) -> (Array:Array of Names), self=@n5)
  ]}
 "Init Max Radius and Num Samples " {
  path="Scatter.K2Node_Composite_2.Init Max Radius and Num Samples "
  nodes [
   Inputs!() -> (Exec:Exec)
   > n1 = For Each Loop(Array:Array of S Selection Choices Scatter Objects Structures=Get Assets To Scatter()) -> (Array Element:S Selection Choices Scatter Objects Structure)
   n2 = Sequence!(execute=@n1.Completed) -> (then_0, then_1)
   n3 = Get Settings()
   n4 = BreakVector2D(InVec=Get Scale Range!(self:BP STPS Scatter Objects Object Reference=@n3) -> (Scale Range:Vector 2D Structure))
   n5 = Get MaxRadius()
   > Set MaxRadius(MaxRadius=SelectFloat(A=10, B=float * float(A=@n5, B=@n4.Y) {is=title K2Node_PromotableOperator}, bPickA=Get Allow Overlap(self=@n3)))
   Outputs!(Completed:Exec=@n2.then_1)
   n6 = Sequence!(execute=@n1.LoopBody) -> (then_0, then_1, then_2:Exec)
   n7 = Break S Selection Choices Scatter Objects(S_SelectionChoices_ScatterObjects=@n1)
   n8 = GetBoundingBox!(self:Static Mesh Object Reference=@n7) -> (ReturnValue:Box Structure)
   n9 = BreakVector(InVec=vector * vector(A=GetBoxCenterSize!(Box:Box Structure=@n8) -> (Center:Vector, Dimensions:Vector).Dimensions, B="0.5,0.5,0.5") -> (ReturnValue=~))
   n10 = "Max (Float)"(A=@n9, B=@n9.Y)
   > Branch(Condition="float > float"(A=@n10, B=Get MaxRadius()))
   > Set MaxRadius(MaxRadius=@n10)
   n11 = "To Float (Integer)"(InInt=@n7.Weight_5_#2)
   n12 = Get Scatter Amount!(self:BP STPS Scatter Objects Object Reference=Get Settings()) -> (Scatter Amount:int=0)
   n13 = Floor(A=float * float(A="To Float (Integer)"(InInt=@n12), B=SafeDivide(A=@n11, B=100))) {is=title K2Node_CallFunction}
   Set Num Samples This Asset(execute=@n6.then_1, Num Samples This Asset:int=@n13) -> (Output_Get:int=0)
   > Set Num Samples This Click(Num Samples This Click=int + int(A=@n13, B=Get Num Samples This Click() -> (Num Samples This Click=0))) -> (Output_Get=0)
   For Loop(execute=@n6.then_2, FirstIndex=1, LastIndex=Get Num Samples This Asset() -> (Num Samples This Asset:int=0))
   > Add(
    TargetArray:Array of S Selection Choices Scatter Objects Structures=Get Asset Sample Counts()
    NewItem:"S Selection Choices Scatter Objects Structure (by ref)"=@n1
   )
  ]
 }
 Set Base Shape Preview Look {nodes [
   Set Base Shape Preview Look() {is=title K2Node_FunctionEntry}
   > SetTextureParameterValue(
    self=Get PreviewShapeDynMat() -> (PreviewShapeDynMat:Material Instance Dynamic Object Reference)
    ParameterName=CanvasFront
    Value=/Game/UI/Foundation/Banner/Banner-SquareLobbyFrame
   )
   n1 = Get Preview Actor DynMeshComp()
   > Is Valid(InputObject=@n1)
   > n2 = GetDynamicMesh!(execute, self:Base Dynamic Mesh Component Object Reference=@n1) -> (then, ReturnValue:Dynamic Mesh Object Reference)
   > SetMeshConstantVertexColor!(
    execute
    TargetMesh:Dynamic Mesh Object Reference=@n2
    Color:Linear Color Structure=Get Scatter Shape Color!(self:BP STPS Scatter Objects Object Reference=Get Settings()) -> (Scatter Shape Color:Linear Color Structure)
    Flags_bRed:bool=True
    Flags_bGreen:bool=True
    Flags_bBlue:bool=True
    Flags_bAlpha:bool=True
    bClearExisting:bool=true
   ) -> (then, ReturnValue:Dynamic Mesh Object Reference)
  ]}
 Set Overlay Dropper Appearance {nodes [
   n1 = Set Overlay Dropper Appearance!() -> (then, Specified Color:Linear Color Structure, InSize:Vector 2D Structure, InPosition:Vector 2D Structure, Denied:bool=~) {is=title K2Node_FunctionEntry}
   > n2 = Sequence!(execute) -> (then_0, then_1)
   > n3 = Branch(Condition=@n1.Denied)
   > n4 = Set CurOverlayDropper(CurOverlayDropper:Texture 2D Object Reference=Get Eyedropper Denied Texture() -> (Eyedropper Denied Texture:Texture 2D Object Reference)) -> (Output_Get:Texture 2D Object Reference)
   n5 = Get CurOverlayDropper() -> (CurOverlayDropper:Texture 2D Object Reference)
   n6 = "To Float (Integer)"(InInt=GetSizeX!(self:Texture 2D Object Reference=@n5) -> (ReturnValue:int=0))
   n7 = "To Float (Integer)"(InInt=GetSizeY!(self:Texture 2D Object Reference=@n5) -> (ReturnValue:int=0))
   SetBrush(
    execute=[@n8.then, @n4.then]
    self=Get Dropper Image() -> (Dropper Image:Image Object Reference)
    InBrush=Make Slate Brush!(
     TintColor:Slate Color Structure=Make Slate Color(SpecifiedColor=@n1)
     DrawAs:ESlateBrushDrawType Enum=Image
     Tiling:ESlateBrushTileType Enum=NoTile
     Mirroring:ESlateBrushMirrorType Enum=NoMirror
     ImageSize:Vector 2D Structure=MakeVector2D(X=@n6, Y=@n7)
     Margin:Margin Structure=(Left=0,Top=0,Right=0,Bottom=0)
     ResourceObject:Object Reference=Get CurOverlayDropper()
     OutlineSettings:Slate Brush Outline Settings Structure=(CornerRadii=(X=0,Y=0,Z=0,W=0),Color=(SpecifiedColor=(R=0,G=0,B=0,A=0),ColorUseRule=UseColor_Specified),Width=0,RoundingType=HalfHeightRadius,bUseBrushTransparency=False)
    ) -> (SlateBrush:Slate Brush Structure) {is=title K2Node_MakeStruct}
   )
   n8 = Set CurOverlayDropper(execute=@n3.else, CurOverlayDropper=Get Eyedropper Texture() -> (Eyedropper Texture:Texture 2D Object Reference))
   n9 = Cast To CanvasPanelSlot!(execute=@n2.then_1, Object=Get Slot!(self:Widget Object Reference=Get Dropper Image()) -> (Slot:Panel Slot Object Reference)) -> (then, CastFailed, AsCanvas Panel Slot:Canvas Panel Slot Object Reference) {is=title K2Node_DynamicCast}
   > SetSize!(execute, self:Canvas Panel Slot Object Reference=@n9, InSize:Vector 2D Structure=@n1.InSize) -> (then)
   n10 = BreakVector2D(InVec=@n1.InSize)
   n11 = BreakVector2D(InVec=@n1.InPosition)
   > SetPosition!(
    execute
    self:Canvas Panel Slot Object Reference=@n9
    InPosition:Vector 2D Structure=MakeVector2D(X=float + float(A=@n11), Y=float - float!(A=@n11.Y, B=@n10.Y, C:double) -> (ReturnValue))
   ) -> (then)
  ]}
 SingleClickCaptureCheck {nodes [
   SingleClickCaptureCheck!() -> (then, InputDeviceState:"Input Device State Structure (by ref)") {is=title K2Node_FunctionEntry}
   > Return Node!(execute, ReturnValue:bool=OR Boolean(A=Get Do Scatter Hotkey Down(), B=Get AddObjectHotkeyDown()))
  ]}
 Update Info {nodes [
   Update Info() {is=title K2Node_FunctionEntry}
   n1 = Get Readout Line Actor()
   > "SetText (Text)"(
    self=Get Value!(self:BP UW Readout Line Object Reference=@n1) -> (Value:Common Text Object Reference)
    InText="To Text (String)"(InString=Get CurrentHoverActorLabel() -> (CurrentHoverActorLabel:String))
   )
   > "SetText (Text)"(self=Get LabelEntry(self=@n1), InText=Get Readout Line Actor Label())
  ]}
]
variables [
 Actor Line Label {type=string, default=""}
 Actor Line Value {type=string, default=""}
 Actors To Ignore In Object Selection Trace {type="array<Actor>"}
 Added To List Message {type=string, default="Added mesh {0}"}
 AddObjectHotkey {type=Key, default=C}
 AddObjectHotkeyDown {type=boolean}
 All Actors At Start {type="array<Actor>"}
 Already In List Message {type=string, default="Mesh {0} is already in your Objects To Scatter list."}
 Always Disallowed Actor Labels {type="array<string>", default [landscape]}
 Assets To Scatter {type="array<S_SelectionChoices_ScatterObjects>", default []}
 Base Shape Change Properties {type="array<string>", default [Base Shape, Scatter Shape Size]}
 Base Shape Mesh {type=DynamicMesh, default=null}
 Base Shape Preview Actor {type=Actor, default=null}
 Circle Shape Steps {type=integer, default=64}
 Clear Warning After Time {type=number, default=5}
 ClickSeed {type=integer}
 Collide Under Landscape {type=boolean, default=true}
 Cur Click Random Stream {type=RandomStream, default {initialSeed=0, seed=0}}
 CurrentHoverActorLabel {type=string, default=""}
 CurTime {type=number}
 Disallowed Actor Message {type=string, default="WARNING: {0} cannot be instanced as it is a landscape object."}
 Disallowed Material Base Message {type=string, default="WARNING: {0} cannot be instanced as its material is unsupported for ISM/HISM."}
 Disallowed Material Message {type=string, default=""}
 Disallowed Project Location Message {type=string, default="WARNING: {0} cannot be instanced as it is not in the Project folder."}
 Do Scatter Hotkey Down {type=boolean}
 DoScatterHotkey {type=Key, default=X}
 Dropper Image {type=Image, default=null}
 Dropper Size {type=Vector2D, default=<16 16>}
 Eyedropper Denied Texture {type=Texture2D, default=/Game/Blueprints/Blutilities/ToolsWrkshp/ScrTools/Shared/SharedWidgets/graphics/denied_red_white_bg}
 Eyedropper Texture {type=Texture2D, default=/Game/Blueprints/Blutilities/ToolsWrkshp/ScrTools/Shared/SharedWidgets/graphics/EyeDropperCursor}
 HaveAlreadyInitHeader {type=boolean}
 Header Widget {type=BP_UW_SharedHeaderWidget_C}
 Hover Out Hit {
  type=HitResult
  default {
   bBlockingHit=false
   bStartPenetrating=false
   boneName=None
   component=null
   distance=0
   elementIndex=0
   faceIndex=0
   item=0
   myBoneName=None
   myItem=-1
   penetrationDepth=0
   physMaterial=null
   time=1
   hitObjectHandle {referenceObject=null}
   impactNormal=<0 0 0>
   impactPoint=<0 0 0>
   location=<0 0 0>
   normal=<0 0 0>
   traceEnd=<0 0 0>
   traceStart=<0 0 0>
  }
 }
 Instanceable Meshes To Scatter {type="array<S_SelectionChoices_ScatterObjects>", default []}
 ISM Actor {type=Actor, default=null}
 Last Clear Time {type=number}
 Message No Asset Path Message {type=string, default=""}
 Message No Asset Path Title {type=string, default=""}
 Mounted Asset Path {type=string, default=""}
 Non Instanceable Meshes To Scatter {type="array<S_SelectionChoices_ScatterObjects>", default []}
 Out Hit {
  type=HitResult
  default {
   bBlockingHit=false
   bStartPenetrating=false
   boneName=None
   component=null
   distance=0
   elementIndex=0
   faceIndex=0
   item=0
   myBoneName=None
   myItem=-1
   penetrationDepth=0
   physMaterial=null
   time=1
   hitObjectHandle {referenceObject=null}
   impactNormal=<0 0 0>
   impactPoint=<0 0 0>
   location=<0 0 0>
   normal=<0 0 0>
   traceEnd=<0 0 0>
   traceStart=<0 0 0>
  }
 }
 Overlay Widget {type=BP_UW_ScatterObjects_Overlay_C, default=null}
 Preview Actor DynMeshComp {type=DynamicMeshComponent, default=null}
 PreviewShapeDynMat {type=MaterialInstanceDynamic}
 Project Handle {type=ValkyrieProjectHandle, default {}}
 Prop Set Save Key {type=string, default=Prop Set Save Key}
 Readout Box {type=BP_UW_ExpandableBox_C}
 Readout Line Actor {type=BP_UW_ReadoutLine_C, default=null}
 Readout Line Actor Label {type=string, default=Actor}
 Scatter Tool Actor Tag {type=string, default=Fortnite Tools Scatter}
 Scattered Assets Inst {type="array<S_InstData_ScatterObjects>", default []}
 Scattered Assets Non Inst {type="array<S_InstData_ScatterObjects>", default []}
 SelectedISMActorsOnStartup {type="array<Actor>"}
 Settings {type=BP_STPS_ScatterObjects_C, default=null}
 Shared State Settings {type=BP_STPS_SharedWidgetState_C, default=null}
 Shortcut Box {type=BP_UW_ExpandableBox_C}
 Shortcut Map {type="map<string>", default {Add Object To Scatter=Hold C + Click, Scatter=Hold X + Click}}
 Shortcut Map Keys {type="array<string>", default [Add Object To Scatter, Scatter]}
 ShortcutLines {type="array<BP_UW_ReadoutLine_C>", default []}
 Target Box Line Set {type=ScriptableToolLineSet, default=null}
 Target Box Lines {type="array<ScriptableToolLine>", default []}
 Undo Clear Context {type=string, default=Scatter}
 Undo Clear Description {type=string, default=Delete Actors}
 Warn Non Inst Message {type=string, default="Note: Some Materials can not be instanced, they will appear with the default Material in the scatter."}
]
